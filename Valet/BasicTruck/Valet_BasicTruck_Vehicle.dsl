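; Blueprint for the Valet basic truck vehicle, tuning its pontoon tick rate scaling when constructed.
; The construction script runs the parent's script first, then overrides separate front and rear pontoon tick rate scales.
; Both values are evaluated from hotfix ScalableFloat variables.
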(fn ConstructionScript
  (|CallParentFunction)
  (Variables|PontoonScaling|SetScalePontoonTickRateFront (ScalableFloat|GetValueatLevel (Variables|Default|GetHotfixScalePontoonTickRateFront_0)))
  (Variables|PontoonScaling|SetScalePontoonTickRateRear (ScalableFloat|GetValueatLevel (Variables|Default|GetHotfixScalePontoonTickRateRear_0))))
